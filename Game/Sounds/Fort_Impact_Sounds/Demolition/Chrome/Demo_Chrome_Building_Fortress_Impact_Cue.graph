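# Sound cue for chrome fortress building impacts in the demolition set, layering a chrome pickaxe hit with a rock impact.
# A mixer plays one of eight chrome waves at full level alongside one of four rock waves, which is quieter, pitched up and slightly delayed.
# An outer modulator varies the combined result.
# Small-object hit attenuation and an object-impact concurrency set control falloff and voice count.
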SoundCue /Game/Sounds/Fort_Impact_Sounds/Demolition/Chrome/Demo_Chrome_Building_Fortress_Impact_Cue
root=Modulator_0
nodes [
 {id=Delay_0, children [Modulator_1], properties {delayMax=0.08, delayMin=0.04}}
 {id=Mixer_0, children [Delay_0,Random_2], properties {inputVolume [0.7,1]}}
 {id=Modulator_0, children [Mixer_0]}
 {id=Modulator_1, children [Random_16], properties {pitchMax=1.6, pitchMin=1.4}}
 {id=Random_16, children [WavePlayer_177, WavePlayer_180, WavePlayer_181, WavePlayer_178], properties {weights [1,1,1,1]}}
 {
  id=Random_2
  children [WavePlayer_16, WavePlayer_17, WavePlayer_18, WavePlayer_19, WavePlayer_20, WavePlayer_21, WavePlayer_22, WavePlayer_23]
  properties {bShouldExcludeFromBranchCulling=true, weights [1,1,1,1,1,1,1,1]}
 }
 {id=WavePlayer_16, properties {soundWaveAssetPtr=&/Chrome_Pickaxe_Impact_Default_01}}
 {id=WavePlayer_17, properties {soundWaveAssetPtr=&/Chrome_Pickaxe_Impact_Default_02}}
 {id=WavePlayer_177, properties {soundWaveAssetPtr=/Game/Sounds/Fort_Impact_Sounds/Demolition/Defaults/rock_impact_stereo_03}}
 {id=WavePlayer_178, properties {soundWaveAssetPtr=/Game/Sounds/Fort_Impact_Sounds/Demolition/Defaults/rock_impact_stereo_04}}
 {id=WavePlayer_18, properties {soundWaveAssetPtr=&/Chrome_Pickaxe_Impact_Default_03}}
 {id=WavePlayer_180, properties {soundWaveAssetPtr=/Game/Sounds/Fort_Impact_Sounds/Demolition/Defaults/rock_impact_stereo_06}}
 {id=WavePlayer_181, properties {soundWaveAssetPtr=/Game/Sounds/Fort_Impact_Sounds/Demolition/Defaults/rock_impact_stereo_07}}
 {id=WavePlayer_19, properties {soundWaveAssetPtr=&/Chrome_Pickaxe_Impact_Default_04}}
 {id=WavePlayer_20, properties {soundWaveAssetPtr=&/Chrome_Pickaxe_Impact_Default_05}}
 {id=WavePlayer_21, properties {soundWaveAssetPtr=&/Chrome_Pickaxe_Impact_Default_06}}
 {id=WavePlayer_22, properties {soundWaveAssetPtr=&/Chrome_Pickaxe_Impact_Default_07}}
 {id=WavePlayer_23, properties {soundWaveAssetPtr=&/Chrome_Pickaxe_Impact_Default_08}}
]
properties {
 attenuationSettings=/Game/Sounds/Attenuation/Fort_Demolition_Hit_Small_Sized_Object_Att
 maxDistance=4000
 soundClassObject=&Fort_Impacts
 volumeMultiplier=0.8
 concurrencyOverrides {resolutionRule=PreventNew}
 concurrencySet [&CONC_Object_Impacts_FourMax]
}
