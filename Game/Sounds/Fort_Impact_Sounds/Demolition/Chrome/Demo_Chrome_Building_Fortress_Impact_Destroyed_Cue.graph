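# Sound cue for the destruction of a chrome fortress building piece, set up as a small-object demolition impact.
# A mixer layers an immediate random chrome pickaxe hit with two successively quieter, delayed rock impacts, forming a staggered debris tail.
# Both rock layers share one random pool through separate pitch-raising modulators, each behind its own slightly randomised delay.
# It plays in the impacts sound class with small-object demolition attenuation and a four-max object impact concurrency set.
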SoundCue /Game/Sounds/Fort_Impact_Sounds/Demolition/Chrome/Demo_Chrome_Building_Fortress_Impact_Destroyed_Cue
root=Mixer_0
nodes [
 {id=Delay_4, children [Modulator_10], properties {delayMax=0.06, delayMin=0.04}}
 {id=Delay_5, children [Modulator_11], properties {delayMax=0.25, delayMin=0.2}}
 {id=Mixer_0, children [Modulator_0,Delay_4,Delay_5], properties {inputVolume [1,0.7,0.5]}}
 {id=Modulator_0, children [Random_2]}
 {id=Modulator_10, children [Random_5], properties {pitchMax=1.5, pitchMin=1.3}}
 {id=Modulator_11, children [Random_5], properties {pitchMax=1.5, pitchMin=1.3}}
 {
  id=Random_2
  children [WavePlayer_16, WavePlayer_17, WavePlayer_18, WavePlayer_19, WavePlayer_20, WavePlayer_21, WavePlayer_22, WavePlayer_23]
  properties {bShouldExcludeFromBranchCulling=true, weights [1,1,1,1,1,1,1,1]}
 }
 {id=Random_5, children [WavePlayer_63, WavePlayer_64, WavePlayer_65], properties {weights [1,1,1]}}
 {id=WavePlayer_16, properties {soundWaveAssetPtr=&/Chrome_Pickaxe_Impact_Default_01}}
 {id=WavePlayer_17, properties {soundWaveAssetPtr=&/Chrome_Pickaxe_Impact_Default_02}}
 {id=WavePlayer_18, properties {soundWaveAssetPtr=&/Chrome_Pickaxe_Impact_Default_03}}
 {id=WavePlayer_19, properties {soundWaveAssetPtr=&/Chrome_Pickaxe_Impact_Default_04}}
 {id=WavePlayer_20, properties {soundWaveAssetPtr=&/Chrome_Pickaxe_Impact_Default_05}}
 {id=WavePlayer_21, properties {soundWaveAssetPtr=&/Chrome_Pickaxe_Impact_Default_06}}
 {id=WavePlayer_22, properties {soundWaveAssetPtr=&/Chrome_Pickaxe_Impact_Default_07}}
 {id=WavePlayer_23, properties {soundWaveAssetPtr=&/Chrome_Pickaxe_Impact_Default_08}}
 {id=WavePlayer_63, properties {soundWaveAssetPtr=/Game/Sounds/Fort_Impact_Sounds/Demolition/Defaults/Fort_Impact_RockFinal_01}}
 {id=WavePlayer_64, properties {soundWaveAssetPtr=/Game/Sounds/Fort_Impact_Sounds/Demolition/Defaults/Fort_Impact_RockFinal_02}}
 {id=WavePlayer_65, properties {soundWaveAssetPtr=/Game/Sounds/Fort_Impact_Sounds/Demolition/Defaults/Fort_Impact_RockFinal_03}}
]
properties {
 attenuationSettings=/Game/Sounds/Attenuation/Fort_Demolition_Hit_Small_Sized_Object_Att
 maxDistance=4000
 soundClassObject=&Fort_Impacts
 volumeMultiplier=0.8
 concurrencyOverrides {resolutionRule=PreventNew}
 concurrencySet [&CONC_Object_Impacts_FourMax]
}
