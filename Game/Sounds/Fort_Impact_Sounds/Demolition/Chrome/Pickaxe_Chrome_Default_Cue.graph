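# Sound cue for the default Chrome pickaxe impact, classed under Fort_Impacts.
# An equally weighted Random node picks one of eight impact waves and feeds it through a Modulator left at its usual settings.
# Small-object demolition hit attenuation handles falloff.
# An overridden concurrency rule caps simultaneous copies, stopping the lowest priority and scaling volume by priority.
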SoundCue /Game/Sounds/Fort_Impact_Sounds/Demolition/Chrome/Pickaxe_Chrome_Default_Cue
root=Modulator_3
nodes [
 {id=Modulator_3, children [Random_1]}
 {
  id=Random_1
  children [WavePlayer_0, WavePlayer_5, WavePlayer_6, WavePlayer_7, WavePlayer_8, WavePlayer_9, WavePlayer_10, WavePlayer_11]
  properties {bShouldExcludeFromBranchCulling=true, weights [1,1,1,1,1,1,1,1]}
 }
 {id=WavePlayer_0, properties {soundWaveAssetPtr=&/Chrome_Pickaxe_Impact_Default_01}}
 {id=WavePlayer_10, properties {soundWaveAssetPtr=&/Chrome_Pickaxe_Impact_Default_07}}
 {id=WavePlayer_11, properties {soundWaveAssetPtr=&/Chrome_Pickaxe_Impact_Default_08}}
 {id=WavePlayer_5, properties {soundWaveAssetPtr=&/Chrome_Pickaxe_Impact_Default_02}}
 {id=WavePlayer_6, properties {soundWaveAssetPtr=&/Chrome_Pickaxe_Impact_Default_03}}
 {id=WavePlayer_7, properties {soundWaveAssetPtr=&/Chrome_Pickaxe_Impact_Default_04}}
 {id=WavePlayer_8, properties {soundWaveAssetPtr=&/Chrome_Pickaxe_Impact_Default_05}}
 {id=WavePlayer_9, properties {soundWaveAssetPtr=&/Chrome_Pickaxe_Impact_Default_06}}
]
properties {
 attenuationSettings=/Game/Sounds/Attenuation/Fort_Demolition_Hit_Small_Sized_Object_Att
 bOverrideConcurrency=true
 maxDistance=4000
 soundClassObject=&Fort_Impacts
 volumeMultiplier=0.8
 concurrencyOverrides {maxCount=3, resolutionRule=StopLowestPriority, volumeScale=0.7, volumeScaleMode=Priority}
 concurrencySet [&CONC_Object_Impacts_FourMax]
}
