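# Sound cue for pickaxe impacts on petrified Chrome surfaces, routed to the impacts sound class.
# A random node picks among eight equally weighted impact variations and feeds a modulator at the root.
# It uses small-object demolition hit attenuation and overrides concurrency to cap simultaneous voices, stopping the lowest priority and scaling volume by priority.
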SoundCue /Game/Sounds/Fort_Impact_Sounds/Demolition/Chrome/Pickaxe_Chrome_Petrified_Cue
root=Modulator_3
nodes [
 {id=Modulator_3, children [Random_0]}
 {
  id=Random_0
  children [WavePlayer_1, WavePlayer_2, WavePlayer_3, WavePlayer_4, WavePlayer_12, WavePlayer_13, WavePlayer_14, WavePlayer_15]
  properties {weights [1,1,1,1,1,1,1,1]}
 }
 {id=WavePlayer_1, properties {soundWaveAssetPtr=&/Chrome_Pickaxe_Impact_Petrified_01}}
 {id=WavePlayer_12, properties {soundWaveAssetPtr=&/Chrome_Pickaxe_Impact_Petrified_05}}
 {id=WavePlayer_13, properties {soundWaveAssetPtr=&/Chrome_Pickaxe_Impact_Petrified_06}}
 {id=WavePlayer_14, properties {soundWaveAssetPtr=&/Chrome_Pickaxe_Impact_Petrified_07}}
 {id=WavePlayer_15, properties {soundWaveAssetPtr=&/Chrome_Pickaxe_Impact_Petrified_08}}
 {id=WavePlayer_2, properties {soundWaveAssetPtr=&/Chrome_Pickaxe_Impact_Petrified_02}}
 {id=WavePlayer_3, properties {soundWaveAssetPtr=&/Chrome_Pickaxe_Impact_Petrified_03}}
 {id=WavePlayer_4, properties {soundWaveAssetPtr=&/Chrome_Pickaxe_Impact_Petrified_04}}
]
properties {
 attenuationSettings=/Game/Sounds/Attenuation/Fort_Demolition_Hit_Small_Sized_Object_Att
 bOverrideConcurrency=true
 maxDistance=4000
 soundClassObject=&Fort_Impacts
 volumeMultiplier=0.8
 concurrencyOverrides {maxCount=3, resolutionRule=StopLowestPriority, volumeScale=0.7, volumeScaleMode=Priority}
 concurrencySet [&CONC_Object_Impacts_FourMax]
}
